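; NPC voice component Blueprint that plays tag-selected voice lines on its owner.
; PlayVoiceLine checks a cooldown and tag validity, then queries a tagged sound bank using the requested tag and the owner's gameplay tags.
; The bank comes from the native voice component variable, falling back to a deprecated Blueprint one.
; Sounds play on a supplied audio component, with a settable source effect chain for voice lines, or otherwise at the owner's location.
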(fn DeprecatedSetSoundBank
  (params (SoundBank :type Fort%20Tagged%20Sound%20Bank%20Object%20Reference))
  (Variables|Default|SetDeprecatedBPSoundBank SoundBank))

(fn InternalGetVoiceSoundBank
  (bind _voicesoundbank (Variables|FortAIComponent_Voice|GetVoiceSoundBank)
    (:then
      (return _voicesoundbank))
    (:else
      (return (Variables|Default|GetDeprecatedBPSoundBank)))) ; Use the new code variable if it was set.
)

(fn PlaySoundCue
  (params (SoundToPlay :type Sound%20Cue%20Object%20Reference) (AudioComponent :type Audio%20Component%20Object%20Reference))
  (Utilities|IsValid SoundToPlay
    (:Is%20Valid
      (bind _output_get (Variables|Default|SetAudio%20Component AudioComponent))
      (Utilities|IsValid _output_get
        (:Is%20Valid
          (Audio|Components|Audio|SetSound _output_get SoundToPlay)
          (Audio|Components|Audio|Play _output_get 0.0))
        (:Is%20Not%20Valid
          (Audio|PlaySoundatLocation SoundToPlay (Transformation|GetActorLocation (Components|GetOwner)))))
      (return true))
    (:Is%20Not%20Valid
      (return false))))

(fn PlayVoiceLine
  (params (Gameplay%20Tag :type Gameplay%20Tag%20Structure%20%28by%20ref%29) (AudioComponent :type Audio%20Component%20Object%20Reference))
  (bind _audio_component (Variables|Default|GetAudio%20Component))
  (LocalVariables|SetCurrentVOTag Gameplay%20Tag)
  (if (and (CallFunction|ShouldPlayVoiceLine) (GameplayTags|IsGameplayTagValid (LocalVariables|GetCurrentVOTag)))
    (Variables|Default|SetAudio%20Component AudioComponent)
    (bind _output_get (LocalVariables|SetSoundBankToUse (CallFunction|InternalGetVoiceSoundBank)))
    (Utilities|IsValid _output_get
      (:Is%20Valid
        (LocalVariables|SetSoundToPlay (Class|FortTaggedSoundBank|GetSoundtoPlay (LocalVariables|GetSoundBankToUse) (LocalVariables|GetCurrentVOTag) (GameplayTags|GetOwnedGameplayTags (Utilities|Casting|ToGameplayTagAssetInterface (Components|GetOwner)) :!class /Script/GameplayTags.BlueprintGameplayTagLibrary)))
        (LocalVariables|GetSoundToPlay
          (:then
            (bind _audio_component_2 (Variables|Default|GetAudio%20Component)
              (:then
                (Variables|/Script/Engine.AudioComponent|SetSourceEffectChain (Variables|Default|GetSourceChainToApply) _audio_component_2)
                (Audio|Components|Audio|SetSound _audio_component (LocalVariables|GetSoundToPlay))
                (Audio|Components|Audio|Play _audio_component 0.0))
              (:else
                (Audio|PlaySoundatLocation (LocalVariables|GetSoundToPlay) (Transformation|GetActorLocation (Components|GetOwner)))))
            (Variables|Default|SetLastTriggerTime (Utilities|Time|GetTimeSeconds))
            (return true))
          (:else))))))

(fn SetSourceEffectChainToApply
  (params (SourceEffectChain :type Sound%20Effect%20Source%20Preset%20Chain%20Object%20Reference))
  (Variables|Default|SetSourceChainToApply SourceEffectChain))

(fn ShouldPlayVoiceLine
  (return (> (- (Utilities|Time|GetTimeSeconds) (Variables|Default|GetLastTriggerTime)) (Variables|Default|GetTriggerTimeThreshold))))
