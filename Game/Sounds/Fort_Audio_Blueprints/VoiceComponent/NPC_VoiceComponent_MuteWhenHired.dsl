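; NPC voice component Blueprint that mutes voice lines when the owning NPC is hired.
; It overrides ShouldPlayVoiceLine to check the owner for the Athena.Gameplay.Status.Converted gameplay tag.
; A line plays only if the tag is absent and the parent class's check also allows it.
; The BeginPlay and Tick events are empty.
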(event EventBeginPlay)

(event EventTick
  (params (DeltaSeconds :type Float%20%28single-precision%29)))

(fn ShouldPlayVoiceLine
  (bind _returnvalue (GameplayTags|HasMatchingGameplayTag(Message) (Components|GetOwner) "(TagName=\"Athena.Gameplay.Status.Converted\")"))
  (return (and (not _returnvalue) (|CallParentFunction))))
